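# Physics asset for the fox pet backpack carrier, giving its dangling decorations simulated secondary motion.
# A large root capsule anchors smaller capsules on three three-bone acorn chains, a three-bone branch chain and three separate leaves.
# Each chain constrains link to parent back to root, with acorns swinging most freely and the branch and leaves held to tighter swing and twist limits.
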PhysicsAsset /BRCosmetics/Accessories/FORT_Backpacks/Backpack_Pet_Fox/Meshes/Fox_Carrier_Physics
bodies {
 dyn_acorn_b_01 [{shapeType=Capsule, length=0.505, radius=0.505, center=<1.687 0.0235 -0.0297>, rotation=<rot 90 0 0>}]
 dyn_acorn_b_02 [{shapeType=Capsule, length=2.7702, radius=0.6924, center=<1.5427 0 0>, rotation=<rot -90 0 0>}]
 dyn_acorn_b_03 [{shapeType=Capsule, length=3.0628, radius=2.0499, center=<2.7373 0.2644 -0.0916>, rotation=<rot 90 0 0>}]
 dyn_acorn_l_01 [{shapeType=Capsule, length=0.505, radius=0.505, center=<2.9751 -0.0486 0.0023>, rotation=<rot 90 0 0>}]
 dyn_acorn_l_02 [{shapeType=Capsule, length=0.505, radius=0.505, center=<1.5427 0 0>, rotation=<rot 90 -0.0099 -0.0099>}]
 dyn_acorn_l_03 [{shapeType=Capsule, length=3.0628, radius=2.0499, center=<3.0623 0.1924 -0.0595>, rotation=<rot 90 0 0>}]
 dyn_acorn_r_01 [{shapeType=Capsule, length=0.505, radius=0.505, center=<1.914 0.1679 -0.0809>, rotation=<rot 90 0.0049 0.0049>}]
 dyn_acorn_r_02 [{shapeType=Capsule, length=0.505, radius=0.505, center=<2.2222 0.0915 0>, rotation=<rot 90 0.0074 0.0074>}]
 dyn_acorn_r_03 [{shapeType=Capsule, length=3.0628, radius=2.0499, center=<2.6599 0.2856 -0.1012>, rotation=<rot 90 0 0>}]
 dyn_branch_01 [{shapeType=Capsule, length=2.5908, radius=0.9536, center=<2.6699 0.3293 0.2681>, rotation=<rot 90 0 0>}]
 dyn_branch_02 [{shapeType=Capsule, length=0.7241, radius=0.6194, center=<1.2869 0 0>, rotation=<rot 90 0 0>}]
 dyn_branch_03 [{shapeType=Capsule, length=0.6145, radius=0.5576, center=<1.3927 0 0>, rotation=<rot 90 0 0>}]
 dyn_leaf_01_l [{shapeType=Capsule, length=2.2017, radius=1.396, center=<3.3217 -0.0083 -0.1042>, rotation=<rot 85.4186 0 0>}]
 dyn_leaf_02_l [{shapeType=Capsule, length=2.728, radius=1.4969, center=<2.9058 -0.3671 0.2789>, rotation=<rot 90 0 0>}]
 dyn_leaf_r [{
   shapeType=Capsule
   length=2.9101
   radius=0.9037
   shapeName=dyn_leaf_r_capsule
   center=<2.7054 -0.2092 -0.46>
   rotation=<rot 0 -99.5316 -87.1547>
  }]
 root [{shapeType=Capsule, length=21.8302, radius=18.62, center=<1.2448 13.445 -13.4577>, rotation=<rot 0 0 -0.537>}]
}
constraints [
 {bone1Name=dyn_acorn_b_01, bone2Name=root, swing1LimitDegrees=45, swing2LimitDegrees=45, twistLimitDegrees=15}
 {bone1Name=dyn_acorn_b_02, bone2Name=dyn_acorn_b_01, swing1LimitDegrees=45, swing2LimitDegrees=45, twistLimitDegrees=15}
 {bone1Name=dyn_acorn_b_03, bone2Name=dyn_acorn_b_02, swing1LimitDegrees=45, swing2LimitDegrees=45, twistLimitDegrees=15}
 {bone1Name=dyn_acorn_l_01, bone2Name=root, swing1LimitDegrees=45, swing2LimitDegrees=45, twistLimitDegrees=15}
 {bone1Name=dyn_acorn_l_02, bone2Name=dyn_acorn_l_01, swing1LimitDegrees=45, swing2LimitDegrees=45, twistLimitDegrees=15}
 {bone1Name=dyn_acorn_l_03, bone2Name=dyn_acorn_l_02, swing1LimitDegrees=45, swing2LimitDegrees=45, twistLimitDegrees=15}
 {bone1Name=dyn_acorn_r_01, bone2Name=root, swing1LimitDegrees=45, swing2LimitDegrees=45, twistLimitDegrees=15}
 {bone1Name=dyn_acorn_r_02, bone2Name=dyn_acorn_r_01, swing1LimitDegrees=45, swing2LimitDegrees=45, twistLimitDegrees=15}
 {bone1Name=dyn_acorn_r_03, bone2Name=dyn_acorn_r_02, swing1LimitDegrees=45, swing2LimitDegrees=45, twistLimitDegrees=15}
 {bone1Name=dyn_leaf_01_l, bone2Name=root, swing1LimitDegrees=15, swing2LimitDegrees=15, twistLimitDegrees=3}
 {bone1Name=dyn_leaf_02_l, bone2Name=root, swing1LimitDegrees=30, swing2LimitDegrees=30, twistLimitDegrees=5}
 {bone1Name=dyn_branch_01, bone2Name=root, swing1LimitDegrees=20, swing2LimitDegrees=20, twistLimitDegrees=5}
 {bone1Name=dyn_branch_02, bone2Name=dyn_branch_01, swing1LimitDegrees=10, swing2LimitDegrees=10, twistLimitDegrees=5}
 {bone1Name=dyn_branch_03, bone2Name=dyn_branch_02, swing1LimitDegrees=30, swing2LimitDegrees=30, twistLimitDegrees=5}
 {bone1Name=dyn_leaf_r, bone2Name=root, swing1LimitDegrees=15, swing2LimitDegrees=15, twistLimitDegrees=3}
]
